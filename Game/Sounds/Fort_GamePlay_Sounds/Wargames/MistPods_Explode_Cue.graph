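# Sound cue for mist pod explosion effects in Wargames, assigned to the Fort_Impacts sound class.
# A Random node picks one of four equally weighted explosion waves and feeds a Modulator at default settings for slight variation.
# The cue overrides attenuation for its distance falloff and lowers overall volume slightly.
# Single-instance, prevent-new concurrency values are set, but the concurrency override flag is not shown as enabled.
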SoundCue /Game/Sounds/Fort_GamePlay_Sounds/Wargames/MistPods_Explode_Cue
root=Modulator_0
nodes [
 {id=Modulator_0, children [Random_1]}
 {id=Random_1, children [WavePlayer_2,WavePlayer_3,WavePlayer_4,WavePlayer_5], properties {weights [1,1,1,1]}}
 {id=WavePlayer_2, properties {soundWaveAssetPtr=&/MistPods_Explode_04}}
 {id=WavePlayer_3, properties {soundWaveAssetPtr=&/MistPods_Explode_01}}
 {id=WavePlayer_4, properties {soundWaveAssetPtr=&/MistPods_Explode_02}}
 {id=WavePlayer_5, properties {soundWaveAssetPtr=&/MistPods_Explode_03}}
]
properties {
 bOverrideAttenuation=true
 maxDistance=4500
 soundClassObject=&Fort_Impacts
 volumeMultiplier=0.8
 attenuationOverrides {falloffDistance=3000, attenuationShapeExtents {x=1500}}
 concurrencyOverrides {maxCount=1, resolutionRule=PreventNew}
}
